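# Physics asset for the large male Menace body cosmetic, covering front, back and left/right hip flaps plus the pelvis and upper thighs.
# Each flap is a two-body chain of boxes or capsules, with limited swing and twist, hung from a tiny driver capsule.
# The four drivers and both thigh twist capsules constrain to the pelvis capsule with free rotation, so angular limits apply only along the flap chains.
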PhysicsAsset /BRCosmetics/Characters/Player/Male/Large/Bodies/M_LRG_Menace/Meshes/M_LRG_Menace_Physics
bodies {
 dyn_driver_flap_bck [{shapeType=Capsule, length=0.505, radius=0.505, rotation=<rot 0 -90 0>}]
 dyn_driver_flap_frt [{shapeType=Capsule, length=0.505, radius=0.505, rotation=<rot 0 -90 0>}]
 dyn_driver_hip_flap_l [{shapeType=Capsule, length=0.505, radius=0.505, rotation=<rot 0 -90 0>}]
 dyn_driver_hip_flap_r [{shapeType=Capsule, length=0.505, radius=0.505, rotation=<rot 0 -90 0>}]
 dyn_flap_bck [{shapeType=Box, extentX=5.3407, extentY=11.306, extentZ=27.706, center=<5.1245 0.0667 0.5278>, rotation=<rot 0 -90 0>}]
 dyn_flap_bck_01 [{shapeType=Capsule, length=16.7245, radius=5.5563, center=<8.0833 0.6245 0>, rotation=<rot 0 -90 0>}]
 dyn_flap_frt [{shapeType=Capsule, length=16.6982, radius=3.9771, center=<4.8862 0.6425 0>, rotation=<rot 0 -90 0>}]
 dyn_flap_frt_01 [{shapeType=Capsule, length=14.9449, radius=4.1411, center=<7.6485 -0.1499 0>, rotation=<rot 0 -90 0>}]
 dyn_hip_flap_l [{shapeType=Box, extentX=7.8728, extentY=3.5638, extentZ=23.0724, center=<4.6557 -0.0953 1.9798>, rotation=<rot -9.3055 0 89.9999>}]
 dyn_hip_flap_l_01 [{shapeType=Box, extentX=6.8728, extentY=3.5638, extentZ=23.0724, center=<5.3896 0.0004 0>, rotation=<rot 0 0 89.9999>}]
 dyn_hip_flap_r [{shapeType=Box, extentX=6.8728, extentY=3.5638, extentZ=23.0724, center=<3.7527 0.0001 0>, rotation=<rot 0 0 89.9999>}]
 dyn_hip_flap_r_01 [{shapeType=Box, extentX=6.8728, extentY=3.5638, extentZ=23.0724, center=<5.3896 0.0004 0>, rotation=<rot 0 0 89.9999>}]
 pelvis [{shapeType=Capsule, length=18.6903, radius=11.0107, center=<-4.7961 -1.083 0>}]
 thigh_twist_01_l [{shapeType=Capsule, length=28.8573, radius=10.1221, center=<0 0 1.8464>, rotation=<rot 90 0 0>}]
 thigh_twist_01_r [{shapeType=Capsule, length=27.3573, radius=9.8721, center=<0 0.0002 -1.2418>, rotation=<rot 90 0 0>}]
}
constraints [
 {bone1Name=dyn_hip_flap_l, bone2Name=dyn_driver_hip_flap_l, swing1LimitDegrees=20, swing2LimitDegrees=35, twistLimitDegrees=5}
 {bone1Name=dyn_hip_flap_r, bone2Name=dyn_driver_hip_flap_r, swing1LimitDegrees=20, swing2LimitDegrees=35, twistLimitDegrees=5}
 {bone1Name=dyn_flap_bck, bone2Name=dyn_driver_flap_bck, swing1LimitDegrees=35, swing2LimitDegrees=20, twistLimitDegrees=5}
 {bone1Name=dyn_flap_frt, bone2Name=dyn_driver_flap_frt, swing1LimitDegrees=35, swing2LimitDegrees=35, twistLimitDegrees=10}
 {bone1Name=dyn_flap_bck_01, bone2Name=dyn_flap_bck, swing1LimitDegrees=35, swing2LimitDegrees=20, twistLimitDegrees=5}
 {bone1Name=dyn_flap_frt_01, bone2Name=dyn_flap_frt, swing1LimitDegrees=35, swing2LimitDegrees=35, twistLimitDegrees=10}
 {bone1Name=dyn_hip_flap_r_01, bone2Name=dyn_hip_flap_r, swing1LimitDegrees=20, swing2LimitDegrees=35, twistLimitDegrees=5}
 {bone1Name=dyn_hip_flap_l_01, bone2Name=dyn_hip_flap_l, swing1LimitDegrees=20, swing2LimitDegrees=35, twistLimitDegrees=5}
 {
  bone1Name=thigh_twist_01_l
  bone2Name=pelvis
  swing1LimitDegrees=45
  swing1Motion=Free
  swing2LimitDegrees=45
  swing2Motion=Free
  twistLimitDegrees=45
  twistMotion=Free
 }
 {
  bone1Name=thigh_twist_01_r
  bone2Name=pelvis
  swing1LimitDegrees=45
  swing1Motion=Free
  swing2LimitDegrees=45
  swing2Motion=Free
  twistLimitDegrees=45
  twistMotion=Free
 }
 {
  bone1Name=dyn_driver_flap_frt
  bone2Name=pelvis
  swing1LimitDegrees=45
  swing1Motion=Free
  swing2LimitDegrees=45
  swing2Motion=Free
  twistLimitDegrees=45
  twistMotion=Free
 }
 {
  bone1Name=dyn_driver_flap_bck
  bone2Name=pelvis
  swing1LimitDegrees=45
  swing1Motion=Free
  swing2LimitDegrees=45
  swing2Motion=Free
  twistLimitDegrees=45
  twistMotion=Free
 }
 {
  bone1Name=dyn_driver_hip_flap_r
  bone2Name=pelvis
  swing1LimitDegrees=45
  swing1Motion=Free
  swing2LimitDegrees=45
  swing2Motion=Free
  twistLimitDegrees=45
  twistMotion=Free
 }
 {
  bone1Name=dyn_driver_hip_flap_l
  bone2Name=pelvis
  swing1LimitDegrees=45
  swing1Motion=Free
  swing2LimitDegrees=45
  swing2Motion=Free
  twistLimitDegrees=45
  twistMotion=Free
 }
]
