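# Animation Blueprint for the Menace large male body cosmetic, adding secondary motion to its front flap and belt.
# The parent mesh pose passes through a Control Rig into crouch and standing branches, each with its own pelvis-space RigidBody physics asset, blended by crouch state.
# Crouch corrects bones before simulation, using event-graph thigh yaw to drive the flap; standing corrects after.
# A final Look At aims the right lat bone at the upper arm.
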AnimBlueprint /BRCosmetics/Characters/Player/Male/Large/Bodies/M_LRG_Menace/Meshes/M_LRG_Menace_AnimBP
target_skeleton=&/M_LRG_Menace_Skeleton
graphs [
 AnimGraph {nodes [
   n1 = "Transform (Modify) Bone - Bone: dyn_parent_hip_flap_l"(ComponentPose=Local To Component(LocalPose=Use cached pose 'main'()), Alpha=1, Translation="0,-4,0", Rotation="0,0,-10", Scale="1,1,1") {
    rotationMode=BMM_Additive
    rotationSpace=BCS_BoneSpace
    translationMode=BMM_Additive
    translationSpace=BCS_BoneSpace
    alphaBoolBlend {blendOption=Linear}
    boneToModify {boneName=dyn_parent_hip_flap_l}
    rotation {roll=-10}
    translation {y=-4}
   }
   n2 = "Constraint - dyn_driver_flap_frt (RR)"!(ComponentPose=@n1, Alpha=1, ConstraintWeights_0=0, ConstraintWeights_1:float=1) -> (Pose) {
    alphaBoolBlend {blendOption=Linear}
    boneToModify {boneName=dyn_driver_flap_frt}
    constraintSetup [{transformType=Rotation, targetBone {boneName=thigh_l}}, {transformType=Rotation, targetBone {boneName=thigh_r}}]
    constraintWeights [1,1]
   }
   n3 = "Transform (Modify) Bone - Bone: dyn_driver_flap_frt"(ComponentPose=@n2, Alpha=Get thigh_rot_l(), Translation="0,0,0", Rotation="0,20,50", Scale="1,1,1") {
    rotationMode=BMM_Additive
    rotationSpace=BCS_BoneSpace
    scaleMode=BMM_Additive
    scaleSpace=BCS_BoneSpace
    translationMode=BMM_Additive
    translationSpace=BCS_BoneSpace
    alphaBoolBlend {blendOption=Linear}
    alphaScaleBiasClamp {bMapRange=true, inRange {max=0, min=-100}, outRange {max=0, min=1}}
    boneToModify {boneName=dyn_driver_flap_frt}
    rotation {roll=50, yaw=20}
   }
   n4 = "Transform (Modify) Bone - Bone: dyn_driver_flap_frt"(ComponentPose=@n3, Alpha=Get thigh_rot_r(), Translation="5,0,0", Rotation="0,3,-50", Scale="1,1,1") {
    rotationMode=BMM_Additive
    rotationSpace=BCS_BoneSpace
    scaleMode=BMM_Additive
    scaleSpace=BCS_BoneSpace
    translationMode=BMM_Additive
    translationSpace=BCS_BoneSpace
    alphaBoolBlend {blendOption=Linear}
    alphaScaleBiasClamp {bMapRange=true, inRange {max=0, min=-100}, outRange {max=0, min=1}}
    boneToModify {boneName=dyn_driver_flap_frt}
    rotation {roll=-50, yaw=3}
    translation {x=5}
   }
   n5 = "Transform (Modify) Bone - Bone: dyn_belt_frt_01"(ComponentPose=@n4, Alpha=10, Translation="-9,0,0", Rotation="0,0,0", Scale="1,1,1") {
    rotationMode=BMM_Additive
    rotationSpace=BCS_BoneSpace
    translationMode=BMM_Additive
    translationSpace=BCS_BoneSpace
    alphaBoolBlend {blendOption=Linear}
    alphaScaleBiasClamp {bInterpResult=true, interpSpeedDecreasing=18, interpSpeedIncreasing=18}
    boneToModify {boneName=dyn_belt_frt_01}
    translation {x=-9}
   }
   n6 = "Transform (Modify) Bone - Bone: dyn_belt_frt_01"(ComponentPose=@n5, Alpha=1, Translation="-4,0,0", Rotation="0,0,0", Scale="1,1,1") {
    rotationMode=BMM_Additive
    rotationSpace=BCS_BoneSpace
    translationMode=BMM_Additive
    translationSpace=BCS_BoneSpace
    alphaBoolBlend {blendOption=Linear}
    boneToModify {boneName=dyn_belt_frt_01}
    translation {x=-4}
   }
   n7 = RigidBody(ComponentPose=@n6) {
    bOverrideWorldGravity=true
    overridePhysicsAsset=&/M_LRG_Menace_crouch_Physics
    simulationSpace=BaseBoneSpace
    alphaBoolBlend {blendOption=Linear}
    baseBoneRef {boneName=pelvis}
    componentLinearAccScale=<0.4 0.4 0.8>
    componentLinearVelScale=<0.2 0.2 0.4>
    overrideWorldGravity {z=-980}
   }
   n8 = Component To Local(ComponentPose=@n7)
   n9 = RigidBody(ComponentPose=Local To Component(LocalPose=Use cached pose 'main'()), ExternalForce=Get gravity_override() -> (gravity_override:Vector)) {
    bOverrideWorldGravity=true
    overridePhysicsAsset=&/M_LRG_Menace_Physics
    simulationSpace=BaseBoneSpace
    alphaBoolBlend {blendOption=Linear}
    baseBoneRef {boneName=pelvis}
    componentLinearAccScale=<0.4 0.4 0.8>
    componentLinearVelScale=<0.2 0.2 0.4>
    overrideWorldGravity {z=-980}
   }
   n10 = "Constraint - dyn_driver_flap_frt (RR)"(ComponentPose=@n9, Alpha=1, ConstraintWeights_0=0.5, ConstraintWeights_1=0.5) {
    alphaBoolBlend {blendOption=Linear}
    alphaScaleBiasClamp {bInterpResult=true, interpSpeedDecreasing=20, interpSpeedIncreasing=20}
    boneToModify {boneName=dyn_driver_flap_frt}
    constraintSetup [{transformType=Rotation, targetBone {boneName=thigh_l}}, {transformType=Rotation, targetBone {boneName=thigh_r}}]
    constraintWeights [1,1]
   }
   n11 = "Transform (Modify) Bone - Bone: dyn_flap_frt"(ComponentPose=@n10, Alpha=1, Translation="0,3,0", Rotation="0,-6,0", Scale="1,1,1") {
    rotationMode=BMM_Additive
    rotationSpace=BCS_BoneSpace
    translationMode=BMM_Additive
    translationSpace=BCS_BoneSpace
    alphaBoolBlend {blendOption=Linear}
    alphaScaleBiasClamp {bInterpResult=true, interpSpeedDecreasing=20, interpSpeedIncreasing=20}
    boneToModify {boneName=dyn_flap_frt}
    rotation {yaw=-6}
    translation {y=3}
   }
   n12 = "Transform (Modify) Bone - Bone: dyn_flap_frt_01"(ComponentPose=@n11, Alpha=1, Translation="0,0,0", Rotation="0,-8,0", Scale="1,1,1") {
    rotationMode=BMM_Additive
    rotationSpace=BCS_BoneSpace
    translationMode=BMM_Additive
    translationSpace=BCS_BoneSpace
    alphaBoolBlend {blendOption=Linear}
    alphaScaleBiasClamp {bInterpResult=true, interpSpeedDecreasing=20, interpSpeedIncreasing=20}
    boneToModify {boneName=dyn_flap_frt_01}
    rotation {yaw=-8}
   }
   n13 = Component To Local(ComponentPose=@n12)
   n14 = Get crouch() -> (crouch:bool)
   n15 = Look At - Bone: deform_lattissimus_00_r(
    ComponentPose=Local To Component(LocalPose=Blend Poses by bool(BlendPose_0=@n8, BlendPose_1=@n13, bActiveValue=@n14) {blendType=Linear})
    Alpha=1
   ) {
    alphaBoolBlend {blendOption=Linear}
    boneToModify {boneName=deform_lattissimus_00_r}
    lookAtLocation {x=0}
    lookAtTarget {boneReference {boneName=upperarm_r}}
    lookAt_Axis {axis=<1 0>}
    lookUp_Axis {axis {x=0, z=1}}
   }
   Output Pose(Result=Component To Local(ComponentPose=@n15)) {layerGroup=None}
   n16 = Control Rig(Source=Copy Pose From Mesh() {bUseAttachedParent=true}) {
    alphaBoolBlend {blendOption=Linear}
    controlRigAssetReference {blueprintRigClass=&/M_LRG_Menace_CtrlRig.M_LRG_Menace_CtrlRig_C}
   }
   Save cached pose 'main'(Pose=@n16) {cacheName=main}
  ]}
 EventGraph {nodes [
   Event BlueprintUpdateAnimation()
   > n1 = Sequence!(execute) -> (then_0, then_1, then_2:Exec, then_3:Exec, then_4:Exec, then_5:Exec, then_6:Exec)
   n2 = GetAnimInstance(self=GetOwnerSkeletalMeshComponent())
   Is Valid(exec=@n1.then_2, InputObject=@n2)
   n3 = OR Boolean!(
    A=Montage_IsPlaying(self=@n2, Montage=/BRCosmetics/Animation/Game/MainPlayer/Emotes/HighActivity/Emote_HighActivity_CMM_M)
    B=Montage_IsPlaying(self=@n2, Montage=/BRCosmetics/Animation/Game/MainPlayer/Emotes/RockingChair/Emote_RockingChair_CMM_M)
    C:bool=Montage_IsPlaying(self=@n2, Montage=/BRCosmetics/Animation/Game/MainPlayer/Emotes/RockingChair/Emote_RockingChair_CLM_M)
    D:bool
    E:bool
   ) -> (ReturnValue)
   > Set isCustomEmotePlaying(isCustomEmotePlaying:double=SelectFloat(A=0, B=1, bPickA=@n3)) -> (Output_Get:double)
   > Set isCustomEmotePlaying2(isCustomEmotePlaying2:double=SelectFloat(A=1, B=0, bPickA=@n3)) -> (Output_Get:double)
   n4 = GetDeltaTransformFromRefPose(self=GetOwnerSkeletalMeshComponent(), BoneName=thigh_r, BaseName=pelvis)
   Set thigh_rot_r(execute=@n1.then_1, thigh_rot_r=BreakRotator(InRot=BreakTransform(InTransform=@n4).Rotation).Yaw)
   n5 = Get bIsMoving2D()
   n6 = Get bIsCrouching()
   > Set crouch_moving(crouch_moving:bool=OR Boolean!(A=AND Boolean(A=@n6, B=@n5), B=Get bIsCrouchSprinting()) -> (ReturnValue)) -> (Output_Get:bool)
   > Set crouch(crouch:bool=Get bIsCrouching()) -> (Output_Get:bool)
   > n7 = Branch(Condition=Get bIsSkyDiving())
   > n8 = Branch(Condition=Get bIsParachuteOpened())
   Set gravity_override(execute=[@n8.then, @n7.else], gravity_override="0, 0,-2000")
   Set gravity_override(execute=@n8.else, gravity_override="0, 0,-2000")
   n9 = GetDeltaTransformFromRefPose(self=GetOwnerSkeletalMeshComponent(), BoneName=thigh_l, BaseName=pelvis)
   Set thigh_rot_l(execute=@n1.then_0, thigh_rot_l=BreakRotator(InRot=BreakTransform(InTransform=@n9).Rotation).Yaw)
  ]}
]
variables [
 crouch {type=boolean}
 crouch_moving {type=boolean}
 gravity_override {type=Vector}
 isCustomEmotePlaying {type=number}
 isCustomEmotePlaying2 {type=number}
 thigh_rot_l {type=number}
 thigh_rot_r {type=number}
]
